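# EQS query for the Large Spire boss AI that picks a player target, favoring pawns that are close and in facing alignment.
# It gathers PlayerPawn_Athena actors within a large radius around the querier.
# An inverse-linear 3D distance test scores nearer players higher.
# A dot-product test filters by minimum alignment and scores it at three times the distance weight, so facing outweighs proximity.
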EnvQuery /SpireBoss/AI/EQS/EQS_Large_Spire_ClosestFacingTarget
options [{
  generator {
   name=EnvQueryGenerator_ActorsOfClass_0
   class=EnvQueryGenerator_ActorsOfClass
   properties {
    bAutoSortTests=true
    bCanRunAsync=false
    envQueryResultNormalizationOption=Default
    generateOnlyActorsInRadius=true
    optionName=""
    searchCenter=&EnvQueryContext_Querier
    searchRadius=50000
    searchedActorClass=/Game/Athena/PlayerPawn_Athena.PlayerPawn_Athena_C
   }
  }
  tests [
   EnvQueryTest_Distance_0 {
    class=EnvQueryTest_Distance
    properties {
     distanceTo=&EnvQueryContext_Querier
     filterType=Range
     floatValueMin=0
     scoringEquation=InverseLinear
     scoringFactor=1
     testComment=""
     testMode=Distance3D
     testPurpose=Score
    }
   }
   EnvQueryTest_Dot_0 {
    class=EnvQueryTest_Dot
    properties {
     clampMaxType=None
     clampMinType=None
     filterType=Minimum
     scoreClampMax=0
     scoreClampMin=0
     scoringFactor=3
     testComment=""
     testPurpose=FilterAndScore
     lineA {lineFrom=&EnvQueryContext_Querier}
    }
   }
  ]
 }]
