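# EQS query for the Large Spire boss that ranks candidate points for a projectile source.
# An OnCircle generator places points on an elevated, rotation-oriented arc with no trace or projection, with radius, height and count bound to scalable float providers.
# Three score-only 3D distance tests favor points near the current target, near known targets at double weight, and far from armature locations.
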EnvQuery /SpireBoss/AI/EQS/EQS_Large_Spire_ProjectileSource
options [{
  generator {
   name=EnvQueryGenerator_OnCircle_0
   class=EnvQueryGenerator_OnCircle
   properties {
    arcAngle=120
    bDefineArc=true
    spaceBetween=1000
    arcDirection {dirMode=Rotation}
    circleCenterZOffset {
     dataBinding=&self:EnvQueryGenerator_OnCircle_0.FortAIDataProvider_ScalableFloat_2
     dataField=FloatValue
     defaultValue=1100
    }
    circleRadius {
     dataBinding=&self:EnvQueryGenerator_OnCircle_0.FortAIDataProvider_ScalableFloat_0
     dataField=FloatValue
     defaultValue=800
    }
    numberOfPoints {
     dataBinding=&self:EnvQueryGenerator_OnCircle_0.FortAIDataProvider_ScalableFloat_1
     dataField=IntValue
     defaultValue=10
    }
    projectionData {extentX=0, navigationFilter=null, postProjectionVerticalOffset=0, projectDown=1024, traceMode=None}
    traceData {navigationFilter=null, traceMode=None}
   }
  }
  tests [
   EnvQueryTest_Distance_0 {
    class=EnvQueryTest_Distance
    properties {
     distanceTo=/SpireBoss/AI/EQS/Contexts/EnvQueryContext_Large_Spire_CurrentTarget.EnvQueryContext_Large_Spire_CurrentTarget_C
     filterType=Range
     floatValueMin=0
     scoringEquation=InverseLinear
     scoringFactor=1
     testComment=""
     testMode=Distance3D
     testPurpose=Score
    }
   }
   EnvQueryTest_Distance_1 {
    class=EnvQueryTest_Distance
    properties {
     distanceTo=/SpireBoss/AI/EQS/Contexts/EnvQueryContext_Large_Spire_KnownTargets.EnvQueryContext_Large_Spire_KnownTargets_C
     filterType=Range
     floatValueMin=0
     scoringEquation=InverseLinear
     scoringFactor=2
     testComment=""
     testMode=Distance3D
     testPurpose=Score
    }
   }
   EnvQueryTest_Distance_2 {
    class=EnvQueryTest_Distance
    properties {
     distanceTo=/SpireBoss/AI/EQS/Contexts/EnvQueryContext_Large_Spire_ArmatureLocations.EnvQueryContext_Large_Spire_ArmatureLocations_C
     filterType=Range
     floatValueMin=0
     scoringEquation=Linear
     scoringFactor=1
     testComment=""
     testMode=Distance3D
     testPurpose=Score
    }
   }
  ]
 }]
